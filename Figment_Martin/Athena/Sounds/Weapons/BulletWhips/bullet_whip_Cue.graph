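# Sound cue that plays a randomized bullet whip sound, the whiz-by of passing weapon projectiles.
# A quality level node sends the upper two levels to a random pick of four waves, and the lower two to a two-wave subset that is excluded from branch culling.
# The cue overrides attenuation to disable distance falloff, uses the weapon projectile sound class, and limits instances through the bullet whiz-by concurrency rule.
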SoundCue /Figment_Martin/Athena/Sounds/Weapons/BulletWhips/bullet_whip_Cue
root=QualityLevel_0
nodes [
 {id=QualityLevel_0, children [Random_0,Random_0,Random_1,Random_1]}
 {id=Random_0, children [WavePlayer_1,WavePlayer_2,WavePlayer_3,WavePlayer_4], properties {weights [1,1,1,1]}}
 {id=Random_1, children [WavePlayer_0,WavePlayer_6], properties {bShouldExcludeFromBranchCulling=true, weights [1,1]}}
 {id=WavePlayer_0, properties {soundWaveAssetPtr=/Game/Athena/Sounds/Weapons/BulletWhips/bullet_whip_01}}
 {id=WavePlayer_1, properties {soundWaveAssetPtr=&bullet_whip_01}}
 {id=WavePlayer_2, properties {soundWaveAssetPtr=/Game/Athena/Sounds/Weapons/BulletWhips/bullet_whip_02}}
 {id=WavePlayer_3, properties {soundWaveAssetPtr=/Game/Athena/Sounds/Weapons/BulletWhips/bullet_whip_03}}
 {id=WavePlayer_4, properties {soundWaveAssetPtr=/Game/Athena/Sounds/Weapons/BulletWhips/bullet_whip_04}}
 {id=WavePlayer_6, properties {soundWaveAssetPtr=&bullet_whip_02}}
]
properties {
 bOverrideAttenuation=true
 maxDistance=2097152
 soundClassObject=/Figment_Martin/Sounds/FortSoundClassesAndModes/Fort_WeaponProjectile
 volumeMultiplier=0.8
 attenuationOverrides {bAttenuate=false, stereoSpread=5000}
 concurrencySet [/Game/Sounds/Concurrency_Rules/Athena_BulletWhizBy_Conc]
}
